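# Master material for the Sunlit glider skin, built on shared character functions with switchable panning emissive, crystal, thin-film, metal LUT effects.
# An FX texture yields crescent, crystal and metal masks.
# Progress parameters animate two sphere masks, one shrinking and one expanding, which form a blade emissive mask for tinting emissive color.
# A Crescent Reveal Progress parameter blends a world position offset that pulls masked vertices toward a root position.
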Material /BRCosmetics/Gadgets/Assets/VinderTech_GliderChute/Glider_Sunlit/Materials/M_Glider_Sunlit_MASTER
graph {
 properties {
  bUseMaterialAttributes=true
  bUsedWithMeshDeformer=true
  bUsedWithSkeletalMesh=true
  pixelDepthOffsetMode=PDOM_Legacy
  previewMesh=/BRCosmetics/Gadgets/Assets/VinderTech_GliderChute/Glider_Sunlit/Meshes/Sunlit_Glider
  usedShadingModels=Default Lit
 }
 nodes [
  n1 = MaterialFunctionCall() {materialFunction=&BaseFN_CharacterFunction_PreFX}
  n2 = MaterialFunctionCall(AO=@n1.AO, Distance Blend=@n1.AO, Fresnel=@n1.AO, Gmap=@n1.AO, Pre Skinned Local Position=@n1.AO, Material=NamedRerouteUsage()) {materialFunction=&BaseFN_CharacterFunction_PostFX}
  MaterialFunctionCall(In=@n2.Result) {materialFunction=&CINE_MF_CharPostFX}
  NamedRerouteDeclaration(None=@n1.Pre skinned Local Position) {name=_PreskinnedLocalPos, nodeColor=<rgba 1 0.2109 0 1>}
  n3 = NamedRerouteUsage()
  n4 = VectorParameter() {group=SunlitGlider, parameterName=MaskCenterLocation_01, channelNames {a="", b=Z, g=Y, r=X}, defaultValue=<rgba 0 150 0 1>}
  n5 = ComponentMask(None=@n4) {b=true, g=true, r=true}
  n6 = ScalarParameter() {group=SunlitGlider, parameterName=Crescent Emissive Progress, sliderMax=1, sortPriority=0}
  n7 = Multiply(
   A=MaterialFunctionCall(
    Input=@n6
    Input Low=Constant()
    Input High=Constant() {r=1}
    Target Low=ScalarParameter() {defaultValue=400, group=SunlitGlider, parameterName=SphereMaskRadiusMin_01}
    Target High=ScalarParameter() {group=SunlitGlider, parameterName=SphereMaskRadiusMax_01}
   ) {desc=Shrinking sphere radius, materialFunction=&RemapValueRange}.Result
   B=Sign(None=@n6)
  )
  n8 = Saturate(None=Multiply(
    A=SphereMask(A=@n3, B=@n5, Radius=@n7) {attenuationRadius=120, hardnessPercent=0}
    B=ScalarParameter() {defaultValue=0.8, group=SunlitGlider, parameterName=SphereMaskEdgeSoftness}
   ))
  n9 = Saturate(None=Power(
    Base=NamedRerouteDeclaration(None=@n8) {name=_SphereMask_01, nodeColor=<rgba 1 0 0.2109 1>}
    Exp=ScalarParameter() {defaultValue=5, group=SunlitGlider, parameterName=SphereMaskCenterSoftness}
   ))
  n10 = NamedRerouteUsage()
  n11 = VectorParameter() {group=SunlitGlider, parameterName=MaskCenterLocation_02, channelNames {a="", b=Z, g=Y, r=X}, defaultValue=<rgba 0 140 0 1>}
  n12 = ScalarParameter() {group=SunlitGlider, parameterName=Metal Emissive Progress, sliderMax=1, sortPriority=0}
  n13 = Constant()
  n14 = Constant() {r=1}
  n15 = ScalarParameter() {group=SunlitGlider, parameterName=SphereMaskRadiusMin_02}
  n16 = ScalarParameter() {defaultValue=250, group=SunlitGlider, parameterName=SphereMaskRadiusMax_02}
  n17 = ScalarParameter() {defaultValue=2, group=SunlitGlider, parameterName=SphereMaskEdgeSoftness_02}
  n18 = Power(
   Base=SphereMask(
    A=NamedRerouteUsage()
    B=ComponentMask(None=@n11) {b=true, g=true, r=true}
    Radius=MaterialFunctionCall(Input=@n12, Input Low=@n13, Input High=@n14, Target Low=@n15, Target High=@n16) {desc=Expanding sphere radius, materialFunction=&RemapValueRange}.Result
   ) {attenuationRadius=120, hardnessPercent=0}
   Exp=@n17
  )
  n19 = NamedRerouteDeclaration(None=Saturate(None=Multiply(A=@n18, B=@n17))) {name=_SphereMask_02, nodeColor=<rgba 1 0 0.2109 1>}
  n20 = ScalarParameter() {defaultValue=3, group=SunlitGlider, parameterName=SphereMaskCenterSoftness_02}
  n21 = Add(A=Multiply(A=Add(A=@n9, B=@n10), B=NamedRerouteUsage()), B=Multiply(A=NamedRerouteUsage(), B=Power(Base=@n19, Exp=@n20)))
  n22 = ScalarParameter() {defaultValue=3, group=SunlitGlider, parameterName=EmissiveIntensity}
  NamedRerouteDeclaration(None=Saturate(None=Multiply(A=@n21, B=@n22))) {name=_BladeEmissiveMask, nodeColor=<rgba 1 0.0234 0 1>}
  n23 = TextureObjectParameter() {
   group=Skin FX
   parameterName=SkinFX_Mask
   texture=/BRCosmetics/Gadgets/Assets/VinderTech_GliderChute/Glider_Sunlit/Textures/T_Glider_Sunlit_FX
  }
  n24 = TextureSample(Tex=@n23)
  NamedRerouteDeclaration(None=@n24.RGBA) {name=SkinFX, nodeColor=<rgba 0.0547 0 1 1>}
  NamedRerouteDeclaration(None=@n1.Result) {name=MA_IN, nodeColor=<rgba 0.7344 0 1 1>}
  NamedRerouteDeclaration(None=@n1.Gmap) {name=GMap, nodeColor=<rgba 0 0.9531 1 1>}
  n25 = NamedRerouteUsage()
  n26 = NamedRerouteUsage()
  n27 = NamedRerouteUsage()
  n28 = NamedRerouteUsage()
  n29 = StaticSwitchParameter(
   True=MaterialFunctionCall!(
    Alpha=StaticComponentMaskParameter(None=@n27) {defaultR=true, group=Skin FX, parameterName=PanningEmissive_Channel}
    isVehicle
    Gmap=NamedRerouteUsage()
    UVOffset=StaticComponentMaskParameter(None=@n27) {defaultB=true, group=Skin FX, parameterName=PanningEmissive_UVOffset_Channel}
    Local Normal
    Local Position
    EmissiveColorOverride
    Material=@n28
   ) {materialFunction=/Game/Weapons/WeaponSkins/MaterialFunction/PanningEmissive}.Result
   False=@n28
  ) {group=Skin FX, parameterName=Use Panning Emissive}
  n30 = StaticSwitchParameter(
   True=MaterialFunctionCall!(
    Edge Map=StaticComponentMaskParameter(None=@n26) {defaultR=true, group=Skin FX, parameterName=Crystal Edge Map Channel}
    Material=@n29
    Alpha=StaticComponentMaskParameter(None=@n26) {defaultR=true, group=Skin FX, parameterName=Crystal Channel}
   ) {materialFunction=/Game/Weapons/WeaponSkins/MaterialFunction/Crystal}.Result
   False=@n29
  ) {group=Skin FX, parameterName=Use Crystal}
  n31 = StaticSwitchParameter(
   True=MaterialFunctionCall!(Alpha=StaticComponentMaskParameter(None=@n25) {defaultR=true, group=Skin FX, parameterName=ThinFilm_Channel}, Material Attributes=@n30) {materialFunction=/Game/Weapons/WeaponSkins/MaterialFunction/ThinFilmFunction}.Result
   False=@n30
  ) {group=Skin FX, parameterName=Use Thin Film}
  n32 = MaterialFunctionCall(Mask=NamedRerouteUsage(), Materials=@n31) {materialFunction=&MF_MetalsLUT}
  n33 = GetMaterialAttributes(None=MaterialFunctionCall!(In=StaticSwitchParameter(True=@n32.Result, False=@n31) {group=Skin FX, parameterName=UseMetalLUT}) {materialFunction=/Game/Weapons/WeaponSkins/MaterialFunction/MF_EmissiveFresnel}.Result) {attributeGetTypes [#1]}
  NamedRerouteDeclaration(None=SetMaterialAttributes(MaterialAttributes=@n33.MaterialAttributes, Emissive Color=NamedRerouteUsage()) {attributeSetTypes [#1]}) {name=MA_OUT, nodeColor=<rgba 0.5234 1 0 1>}
  NamedRerouteDeclaration(None=@n33.Emissive Color) {name=Emissive_IN, nodeColor=<rgba 0 0.9297 1 1>}
  n34 = NamedRerouteUsage()
  n35 = VectorParameter() {group=SunlitGlider, parameterName=BladeEmissiveColor, defaultValue=<rgba 0 0.8139 0.9911 1>}
  NamedRerouteDeclaration(None=LinearInterpolate(A=@n34, B=LinearInterpolate(A=@n34, B=@n35.RGB) {constAlpha=0.3}, Alpha=NamedRerouteUsage())) {name=Emissive_OUT, nodeColor=<rgba 0 0.7617 1 1>}
  n36 = Subtract(A=VectorParameter() {group=SunlitGlider, parameterName=Sunlit Glider Root Position, defaultValue=<rgba 0 0 0 1>}.RGB, B=WorldPosition().XYZ)
  n37 = NamedRerouteUsage()
  NamedRerouteDeclaration(None=LinearInterpolate(
    A=Multiply(A=@n36, B=@n37)
    B=NamedRerouteUsage()
    Alpha=ScalarParameter() {defaultValue=1, group=SunlitGlider, parameterName=Crescent Reveal Progress, sortPriority=0}
   ) {constAlpha=1}) {name=WPO_OUT, nodeColor=<rgba 1 0.3516 0 1>}
  NamedRerouteDeclaration() {name=WPO_IN, nodeColor=<rgba 1 0.1641 0 1>}
  NamedRerouteDeclaration(None=@n24.R) {name=_Crescent, nodeColor=<rgba 1 0.6563 0 1>}
  n38 = Add(
   A=NamedRerouteDeclaration(None=SmoothStep(Value=@n24.G) {constMax=0.4}) {name=_Crystal, nodeColor=<rgba 0.5938 0 1 1>}
   B=NamedRerouteUsage()
  )
  NamedRerouteDeclaration(None=SmoothStep(Value=OneMinus(None=@n38)) {constMin=0.8}) {name=_Metal, nodeColor=<rgba 0.5469 1 0 1>}
  NamedRerouteUsage()
 ]
}
